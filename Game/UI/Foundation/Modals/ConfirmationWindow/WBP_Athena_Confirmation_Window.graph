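# Confirmation modal widget for AthenaConfirmationWindow: title, rich-text description, optional icon and dynamic button row over a blurred backdrop, plus a touch-only close button.
# OnShowConfirmationContent selects the main page; a valid icon brush resource shows the icon and left-aligns title and description, otherwise they center.
# BoundAnim_PostConfirm switches to a post-confirmation page and fades in its message and checkmark.
# OnSetButtonText casts each button to GenericModalButton to set its label.
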WidgetBlueprint /Game/UI/Foundation/Modals/ConfirmationWindow/WBP_Athena_Confirmation_Window
animations [BoundAnim_PostConfirm {
  binding_count=2
  length_seconds=1
  object=&self:BoundAnim_PostConfirm
  sequence {bindings [
    VisibilitySwitcher_MainContent {
     id=#1
     possessed_class=CommonVisibilitySwitcher
     tracks [{
       type=MovieSceneIntegerTrack
       display_name=Active Widget Index
       sections [{type=MovieSceneIntegerSection, end_frame=0, start_frame=0, channels [None {type=MovieSceneScriptingIntegerChannel, const [1,1]}]}]
      }]
    }
    VerticalBox_PostConfirmationMessage {
     id=#2
     possessed_class=VerticalBox
     tracks [{
       type=MovieSceneFloatTrack
       display_name=Render Opacity
       sections [{type=MovieSceneFloatSection, end_frame=20, start_frame=0, channels [None {type=MovieSceneScriptingFloatChannel, keys [[0,0],[10,1],[20,1]]}]}]
      }]
    }
   ]}
 }]
graphs [EventGraph {nodes [
   Event OnShowConfirmationContent() {is=title K2Node_Event}
   > n1 = Sequence!(execute) -> (then_0, then_1, then_2:Exec, then_3:Exec)
   > SetActiveWidgetIndex(
    self:Common Visibility Switcher Object Reference=Get VisibilitySwitcher_MainContent() -> (VisibilitySwitcher_MainContent:Common Visibility Switcher Object Reference)
    Index=0
   )
   n2 = Get Image_MainIcon() -> (Image_MainIcon:Fort Lazy Image Object Reference)
   n3 = IsValid(Object=GetBrushResource!(Brush:"Slate Brush Structure (by ref)"=Get Brush!(self:Image Object Reference=@n2) -> (Brush:Slate Brush Structure)) -> (ReturnValue:Object Reference))
   Set Justification!(
    execute=@n1.then_1
    Justification:ETextJustify Enum=Select(Option 0:ETextJustify Enum=Center, Option 1:ETextJustify Enum=Left, Index=@n3) -> (ReturnValue:ETextJustify Enum)
    self:Text Layout Widget Object Reference=Get RichText_Description() -> (RichText_Description:Common Rich Text Block Object Reference)
   ) -> (then, Output_Get:ETextJustify Enum=Left)
   n4 = Cast To VerticalBoxSlot!(
    execute=@n1.then_2
    Object=Get Slot!(self:Widget Object Reference=Get Text_Title() -> (Text_Title:Common Text Object Reference)) -> (Slot:Panel Slot Object Reference)
   ) -> (then, CastFailed, AsVertical Box Slot:Vertical Box Slot Object Reference) {is=title K2Node_DynamicCast}
   > SetHorizontalAlignment(
    self:Vertical Box Slot Object Reference=@n4
    InHorizontalAlignment=Select(Option 0:EHorizontalAlignment Enum=HAlign_Center, Option 1:EHorizontalAlignment Enum=HAlign_Left, Index=@n3) -> (ReturnValue:EHorizontalAlignment Enum)
   )
   SetVisibility(
    execute=@n1.then_3
    self=Get SizeBox_MainIcon() -> (SizeBox_MainIcon:Size Box Object Reference)
    InVisibility=Select(Option 0:ESlateVisibility Enum=Collapsed, Option 1:ESlateVisibility Enum=SelfHitTestInvisible, Index=@n3) -> (ReturnValue:ESlateVisibility Enum)
   )
   n5 = Event OnSetButtonText!() -> (OutputDelegate, then, Button:Common Button Base Object Reference, Text:"Text (by ref)") {is=title K2Node_Event}
   > n6 = Cast To GenericModalButton!(execute, Object=@n5.Button) -> (then, CastFailed, AsGeneric Modal Button:Generic Modal Button Object Reference) {is=title K2Node_DynamicCast}
   > SetButtonText!(execute, self:Generic Modal Button Object Reference=@n6, ButtonText:Text=@n5.Text) -> (then)
   Event PreConstruct()
   Event Construct()
  ]}]
variables []
widgets {
 info {inheritedWidgetCount=7, parentClass=/Script/FortniteUI.AthenaConfirmationWindow, rootWidgetClass=&Overlay, widgetCount=26}
 widgets [
  Overlay {widgetClassPath=&Overlay, widget=Overlay}
  BackgroundBlurLightbox {
   widgetClassPath=&BackgroundBlur
   parent=Overlay
   widget=BackgroundBlurLightbox
   properties {
    blurStrength=8
    lowQualityFallbackBrush {
     bIsDynamicallyLoaded=false
     drawAs=Image
     imageType=NoImage
     mirroring=NoMirror
     resourceName=None
     resourceObject=null
     tiling=NoTile
     imageSize=<0 0>
     margin=<ltrb 0 0 0 0>
     outlineSettings {
      bUseBrushTransparency=false
      roundingType=HalfHeightRadius
      width=0
      color {colorUseRule=UseColor_Specified, specifiedColor=<rgba 0 0 0 0>}
      cornerRadii=<0 0 0 0>
     }
     tintColor {colorUseRule=UseColor_Specified, specifiedColor=<rgba 0 0 0 0.75>}
     uVRegion {bIsValid=false, max=<0 0>, min=<0 0>}
    }
   }
   slot {class=OverlaySlot, properties {horizontalAlignment=HAlign_Fill, verticalAlignment=VAlign_Fill}}
  }
  Lightbox_HitTestBlocker {
   widgetClassPath=&Image
   parent=BackgroundBlurLightbox
   widget=Lightbox_HitTestBlocker
   properties {brush {
     bIsDynamicallyLoaded=false
     drawAs=Image
     imageType=NoImage
     mirroring=NoMirror
     resourceName=None
     resourceObject=/Game/Athena/UI/Frontend/Art/M_UI_RadialGradientMaster_Dark
     tiling=NoTile
     imageSize=<32 32>
     margin=<ltrb 0 0 0 0>
     outlineSettings {
      bUseBrushTransparency=false
      roundingType=HalfHeightRadius
      width=0
      color {colorUseRule=UseColor_Specified, specifiedColor=<rgba 0 0 0 0>}
      cornerRadii=<0 0 0 0>
     }
     tintColor {colorUseRule=UseColor_Specified, specifiedColor=<rgba 0.5972 0.5972 0.5972 1>}
     uVRegion {bIsValid=false, max=<0 0>, min=<0 0>}
    }}
   slot {class=BackgroundBlurSlot, properties {padding=<ltrb 0 0 0 0>}}
  }
  Border_TapToCloseZone {
   widgetClassPath=&CommonBorder
   bInherited=true
   parent=Overlay
   widget=Border_TapToCloseZone
   properties {visibility=Visible}
   slot {class=OverlaySlot, properties {horizontalAlignment=HAlign_Fill, verticalAlignment=VAlign_Fill}}
  }
  Overlay_Modal {
   widgetClassPath=&Overlay
   parent=Overlay
   widget=Overlay_Modal
   slot {class=OverlaySlot, properties {horizontalAlignment=HAlign_Fill, verticalAlignment=VAlign_Center}}
  }
  Containment {
   widgetClassPath=&Image
   parent=Overlay_Modal
   widget=Containment
   properties {
    brush {
     bIsDynamicallyLoaded=false
     drawAs=Image
     imageType=NoImage
     mirroring=NoMirror
     resourceName=None
     resourceObject=/Game/Athena/UI/Subscriptions/Art/M_UI_Subs_Color_FIll_Modals
     tiling=NoTile
     imageSize=<32 32>
     margin=<ltrb 0 0 0 0>
     outlineSettings {
      bUseBrushTransparency=false
      roundingType=HalfHeightRadius
      width=0
      color {colorUseRule=UseColor_Specified, specifiedColor=<rgba 0 0 0 0>}
      cornerRadii=<0 0 0 0>
     }
     tintColor {colorUseRule=UseColor_Specified, specifiedColor=<rgba 1 1 1 1>}
     uVRegion {bIsValid=false, max=<0 0>, min=<0 0>}
    }
    colorAndOpacity=<rgba 1 1 1 0.5>
    renderTransformPivot=<0.5 1>
   }
   slot {class=OverlaySlot, properties {horizontalAlignment=HAlign_Fill, verticalAlignment=VAlign_Fill}}
  }
  VisibilitySwitcher_MainContent {
   widgetClassPath=&CommonVisibilitySwitcher
   bIsVariable=true
   parent=Overlay_Modal
   widget=VisibilitySwitcher_MainContent
   slot {class=OverlaySlot, properties {horizontalAlignment=HAlign_Center}}
  }
  ContentSZ {widgetClassPath=&SafeZone, parent=VisibilitySwitcher_MainContent, widget=ContentSZ, slot {class=CommonVisibilitySwitcherSlot}}
  HorizontalBox_0 {widgetClassPath=&HorizontalBox, parent=ContentSZ, widget=HorizontalBox_0, slot {class=SafeZoneSlot}}
  Overlay_2 {
   widgetClassPath=&Overlay
   parent=HorizontalBox_0
   widget=Overlay_2
   slot {class=HorizontalBoxSlot, properties {padding=<ltrb 0 75 0 75>, size {sizeRule=Fill, value=1}}}
  }
  TextContent {
   widgetClassPath=&VerticalBox
   parent=Overlay_2
   widget=TextContent
   slot {class=OverlaySlot, properties {horizontalAlignment=HAlign_Center, verticalAlignment=VAlign_Fill, padding=<ltrb 0 0 0 50>}}
  }
  Text_Title {
   widgetClassPath=&CommonTextBlock
   bInherited=true
   bIsVariable=true
   parent=TextContent
   widget=Text_Title
   properties {
    justification=Center
    text="{TEXT}"
    textTransformPolicy=ToUpper
    wrapTextAt=1150
    font {
     bForceMonospaced=false
     bMaterialIsStencil=false
     fontMaterial=null
     fontObject=&Burbank
     letterSpacing=0
     monospacedWidth=1
     size=33
     skewAmount=0
     typefaceFontName=BurbankRegular-Black
     outlineSettings {
      bApplyOutlineToDropShadows=false
      bMiteredCorners=false
      bSeparateFillAlpha=false
      outlineMaterial=null
      outlineSize=0
      outlineColor=<rgba 1 1 1 0>
     }
    }
    renderTransform {angle=0, scale=<1 1>, shear=<-12 0>, translation=<0 0>}
    shadowOffset=<0 0>
   }
   slot {class=VerticalBoxSlot, properties {horizontalAlignment=HAlign_Left, padding=<ltrb 0 0 0 20>}}
  }
  RichText_Description {
   widgetClassPath=&CommonRichTextBlock
   bInherited=true
   bIsVariable=true
   parent=TextContent
   widget=RichText_Description
   properties {
    bOverrideDefaultStyle=true
    text="This action clears all data stored for this content.  Only do this if you want to start this content over from the beginning and clear all your progress.\r\n\r\n\r\n\r\n<b>This action cannot be undone!</>"
    textStyleSet=/Game/Athena/UI/Customization/LockerStyles
    wrapTextAt=1100
    defaultTextStyleOverride {
     overflowPolicy=Clip
     transformPolicy=None
     colorAndOpacity {colorUseRule=UseColor_Specified, specifiedColor=<rgba 0.2159 0.6584 1 1>}
     font {
      bForceMonospaced=false
      bMaterialIsStencil=false
      fontMaterial=null
      fontObject=&BurbankSmall
      letterSpacing=0
      monospacedWidth=1
      size=20
      skewAmount=0
      typefaceFontName=Bold
      outlineSettings {
       bApplyOutlineToDropShadows=false
       bMiteredCorners=false
       bSeparateFillAlpha=false
       outlineMaterial=null
       outlineSize=0
       outlineColor=<rgba 0 0 0 1>
      }
     }
     highlightColor {colorUseRule=UseColor_Specified, specifiedColor=<rgba 0 0 0 1>}
     highlightShape {
      bIsDynamicallyLoaded=false
      drawAs=Image
      imageType=NoImage
      mirroring=NoMirror
      resourceName=None
      resourceObject=null
      tiling=NoTile
      imageSize=<32 32>
      margin=<ltrb 0 0 0 0>
      outlineSettings {
       bUseBrushTransparency=false
       roundingType=HalfHeightRadius
       width=0
       color {colorUseRule=UseColor_Specified, specifiedColor=<rgba 0 0 0 0>}
       cornerRadii=<0 0 0 0>
      }
      tintColor {colorUseRule=UseColor_Specified, specifiedColor=<rgba 1 1 1 1>}
      uVRegion {bIsValid=false, max=<0 0>, min=<0 0>}
     }
     selectedBackgroundColor {colorUseRule=UseColor_Foreground, specifiedColor=<rgba 1 0 1 1>}
     shadowColorAndOpacity=<rgba 0 0 0 1>
     shadowOffset=<0 0>
     strikeBrush {
      bIsDynamicallyLoaded=false
      drawAs=Image
      imageType=NoImage
      mirroring=NoMirror
      resourceName=None
      resourceObject=null
      tiling=NoTile
      imageSize=<32 32>
      margin=<ltrb 0 0 0 0>
      outlineSettings {
       bUseBrushTransparency=false
       roundingType=HalfHeightRadius
       width=0
       color {colorUseRule=UseColor_Specified, specifiedColor=<rgba 0 0 0 0>}
       cornerRadii=<0 0 0 0>
      }
      tintColor {colorUseRule=UseColor_Specified, specifiedColor=<rgba 1 1 1 1>}
      uVRegion {bIsValid=false, max=<0 0>, min=<0 0>}
     }
     underlineBrush {
      bIsDynamicallyLoaded=false
      drawAs=Image
      imageType=NoImage
      mirroring=NoMirror
      resourceName=None
      resourceObject=null
      tiling=NoTile
      imageSize=<32 32>
      margin=<ltrb 0 0 0 0>
      outlineSettings {
       bUseBrushTransparency=false
       roundingType=HalfHeightRadius
       width=0
       color {colorUseRule=UseColor_Specified, specifiedColor=<rgba 0 0 0 0>}
       cornerRadii=<0 0 0 0>
      }
      tintColor {colorUseRule=UseColor_Specified, specifiedColor=<rgba 1 1 1 1>}
      uVRegion {bIsValid=false, max=<0 0>, min=<0 0>}
     }
    }
   }
   slot {class=VerticalBoxSlot, properties {horizontalAlignment=HAlign_Center, verticalAlignment=VAlign_Center, size {sizeRule=Automatic, value=0.5}}}
  }
  SizeBox_MainIcon {
   widgetClassPath=&SizeBox
   bIsVariable=true
   parent=HorizontalBox_0
   widget=SizeBox_MainIcon
   properties {bOverride_HeightOverride=true, bOverride_WidthOverride=true, heightOverride=300, widthOverride=300}
   slot {
    class=HorizontalBoxSlot
    properties {horizontalAlignment=HAlign_Left, verticalAlignment=VAlign_Bottom, padding=<ltrb 0 0 86 75>, size {sizeRule=Automatic, value=0.4}}
   }
  }
  Image_MainIcon {
   widgetClassPath=&FortLazyImage
   bInherited=true
   bIsVariable=true
   parent=SizeBox_MainIcon
   widget=Image_MainIcon
   properties {
    brush {
     bIsDynamicallyLoaded=false
     drawAs=Image
     imageType=NoImage
     mirroring=NoMirror
     resourceName=None
     resourceObject=/Game/UI/Foundation/Shell/Textures/T-Icon-Warning-128
     tiling=NoTile
     imageSize=<128 128>
     margin=<ltrb 0 0 0 0>
     outlineSettings {
      bUseBrushTransparency=false
      roundingType=HalfHeightRadius
      width=0
      color {colorUseRule=UseColor_Specified, specifiedColor=<rgba 0 0 0 0>}
      cornerRadii=<0 0 0 0>
     }
     tintColor {colorUseRule=UseColor_Specified, specifiedColor=<rgba 1 1 1 1>}
     uVRegion {bIsValid=false, max=<0 0>, min=<0 0>}
    }
    colorAndOpacity=<rgba 0.007 0.0999 0.4342 1>
    renderTransform {angle=-5, scale=<1 1>, shear=<0 0>, translation=<0 0>}
   }
   slot {class=SizeBoxSlot}
  }
  SizeBox_PostConfirmationMessage {
   widgetClassPath=&SizeBox
   parent=VisibilitySwitcher_MainContent
   widget=SizeBox_PostConfirmationMessage
   properties {bOverride_MinDesiredHeight=true, minDesiredHeight=455}
   slot {class=CommonVisibilitySwitcherSlot, properties {verticalAlignment=VAlign_Center, padding=<ltrb 0 0 64 0>}}
  }
  VerticalBox_PostConfirmationMessage {
   widgetClassPath=&VerticalBox
   parent=SizeBox_PostConfirmationMessage
   widget=VerticalBox_PostConfirmationMessage
   slot {class=SizeBoxSlot, properties {verticalAlignment=VAlign_Center}}
  }
  Text_PostConfirmation {
   widgetClassPath=&CommonTextBlock
   bInherited=true
   bIsVariable=true
   parent=VerticalBox_PostConfirmationMessage
   widget=Text_PostConfirmation
   properties {
    justification=Center
    text="{TEXT}"
    textTransformPolicy=ToUpper
    wrapTextAt=1150
    font {
     bForceMonospaced=false
     bMaterialIsStencil=false
     fontMaterial=null
     fontObject=&Burbank
     letterSpacing=0
     monospacedWidth=1
     size=33
     skewAmount=0
     typefaceFontName=BurbankRegular-Black
     outlineSettings {
      bApplyOutlineToDropShadows=false
      bMiteredCorners=false
      bSeparateFillAlpha=false
      outlineMaterial=null
      outlineSize=0
      outlineColor=<rgba 1 1 1 0>
     }
    }
    renderTransform {angle=0, scale=<1 1>, shear=<-12 0>, translation=<0 0>}
    shadowOffset=<0 0>
   }
   slot {class=VerticalBoxSlot, properties {horizontalAlignment=HAlign_Left, padding=<ltrb 0 0 0 20>}}
  }
  SizeBox_PostConfirmationIcon {
   widgetClassPath=&SizeBox
   bIsVariable=true
   parent=VerticalBox_PostConfirmationMessage
   widget=SizeBox_PostConfirmationIcon
   properties {bOverride_HeightOverride=true, bOverride_WidthOverride=true, heightOverride=180, widthOverride=180}
   slot {
    class=VerticalBoxSlot
    properties {horizontalAlignment=HAlign_Center, verticalAlignment=VAlign_Bottom, padding=<ltrb 0 20 0 0>, size {sizeRule=Automatic, value=0.4}}
   }
  }
  Image_PostConfirmationnIcon {
   widgetClassPath=&FortLazyImage
   bIsVariable=true
   parent=SizeBox_PostConfirmationIcon
   widget=Image_PostConfirmationnIcon
   properties {brush {
     bIsDynamicallyLoaded=false
     drawAs=Image
     imageType=NoImage
     mirroring=NoMirror
     resourceName=None
     resourceObject=/Game/Athena/UI/InGame/Creative/PropertyEditor/Icon_Picker/Textures/128x128/T_UI_IconLibrary_checkmark_128
     tiling=NoTile
     imageSize=<128 128>
     margin=<ltrb 0 0 0 0>
     outlineSettings {
      bUseBrushTransparency=false
      roundingType=HalfHeightRadius
      width=0
      color {colorUseRule=UseColor_Specified, specifiedColor=<rgba 0 0 0 0>}
      cornerRadii=<0 0 0 0>
     }
     tintColor {colorUseRule=UseColor_Specified, specifiedColor=<rgba 0.0343 1 0.227 1>}
     uVRegion {bIsValid=false, max=<0 0>, min=<0 0>}
    }}
   slot {class=SizeBoxSlot}
  }
  Image_BackgroundBottomBar {
   widgetClassPath=&Image
   parent=Overlay_Modal
   widget=Image_BackgroundBottomBar
   properties {brush {
     bIsDynamicallyLoaded=false
     drawAs=Image
     imageType=NoImage
     mirroring=NoMirror
     resourceName=None
     resourceObject=/FortUILibrary/Foundation/Art_ToolKit/AngledBoxes/M_UI_MoreInfo_BottomDividerLIne_Blue
     tiling=NoTile
     imageSize=<32 26>
     margin=<ltrb 0 0 0 0>
     outlineSettings {
      bUseBrushTransparency=false
      roundingType=HalfHeightRadius
      width=0
      color {colorUseRule=UseColor_Specified, specifiedColor=<rgba 0 0 0 0>}
      cornerRadii=<0 0 0 0>
     }
     tintColor {colorUseRule=UseColor_Specified, specifiedColor=<rgba 0.3263 0.3457 0.4149 1>}
     uVRegion {bIsValid=false, max=<0 0>, min=<0 0>}
    }}
   slot {class=OverlaySlot, properties {horizontalAlignment=HAlign_Fill, verticalAlignment=VAlign_Bottom}}
  }
  ButtonHeightSB {
   widgetClassPath=&SizeBox
   parent=Overlay_Modal
   widget=ButtonHeightSB
   properties {bOverride_MaxDesiredHeight=true, maxDesiredHeight=100}
   slot {class=OverlaySlot, properties {horizontalAlignment=HAlign_Center, verticalAlignment=VAlign_Bottom, padding=<ltrb 0 32 0 -44>}}
  }
  EntryBox_Buttons {
   widgetClassPath=&DynamicEntryBox
   bInherited=true
   parent=ButtonHeightSB
   widget=EntryBox_Buttons
   properties {
    entryHorizontalAlignment=HAlign_Center
    entryVerticalAlignment=VAlign_Center
    entryWidgetClass=&/WBP_Athena_ConfirmationButton.WBP_Athena_ConfirmationButton_C
    numDesignerPreviewEntries=2
    entrySpacing=<40 0>
   }
   slot {class=SizeBoxSlot, properties {horizontalAlignment=HAlign_Center, verticalAlignment=VAlign_Bottom}}
  }
  VisibilityBorder_Touch {
   widgetClassPath=&CommonHardwareVisibilityBorder
   parent=Overlay
   widget=VisibilityBorder_Touch
   properties {visibilityQuery {
     autoDescription=" ANY( Input.Touch )"
     tokenStreamVersion=0
     userDescription=""
     queryTokenStream [0,1,1,1,0]
     tagDictionary [{tagName=Input.Touch}]
    }}
   slot {class=OverlaySlot, properties {horizontalAlignment=HAlign_Right, padding=<ltrb 0 32 32 0>}}
  }
  CloseTouch {
   widgetClassPath=&SafeZone
   parent=VisibilityBorder_Touch
   widget=CloseTouch
   slot {class=BorderSlot, properties {padding=<ltrb 0 0 0 0>}}
  }
  Button_CloseTouch {
   widgetClassPath=/FortUILibrary/Foundation/Buttons/CloseButton
   bInherited=true
   bIsVariable=true
   parent=CloseTouch
   widget=Button_CloseTouch
   properties {
    onButtonBaseClicked=(&self:WidgetTree.Button_CloseTouch.HandleOnButtonClicked)
    onButtonBaseDoubleClicked=(&self:WidgetTree.Button_CloseTouch.HandleOnButtonDoubleClicked)
    onButtonBaseHovered=(&self:WidgetTree.Button_CloseTouch.HandleOnButtonHovered)
    onButtonBaseUnhovered=(&self:WidgetTree.Button_CloseTouch.HandleOnButtonUnhovered)
    onSelectedChangedBase=(&self:WidgetTree.Button_CloseTouch.HandleOnSelectedChanged)
   }
   slot {class=SafeZoneSlot}
  }
 ]
}
