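# A UI sound cue that plays the whoosh for closing a generic confirmation toast or window in the front end.
# A single wave player feeds the ConfirmWindow_Close whoosh wave straight to the output, with no randomization or modulation.
# The cue is routed to the Fort_UI sound class with a boosted volume multiplier.
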SoundCue /Game/Sounds/Fort_UI_Sounds/FrontEnd/Whooshes/Toasts/UI_Whoosh_Generic_Confirmation_Close_Cue
root=WavePlayer_0
nodes [{id=WavePlayer_0, properties {soundWaveAssetPtr=/Game/Sounds/Fort_UI_Sounds/FrontEnd/Whooshes/Toasts/wav/UI_Whoosh_ConfirmWindow_Close}}]
properties {maxDistance=2097152, soundClassObject=&Fort_UI, volumeMultiplier=2}
