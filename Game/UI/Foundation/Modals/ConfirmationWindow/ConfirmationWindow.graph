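# Widget Blueprint for the modal confirmation dialog, a FortConfirmationWindow child with title, icon, scrollable description and dynamically filled button rows on a dimmed lightbox.
# Confirm and decline configuration casts buttons to the UIKit CTA base, setting size, input priority and triggering action from a data table.
# Activation plays a sound and a short fade-and-scale intro; outro ends immediately with a sound.
# TapToClose routes primary presses to the decline action.
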WidgetBlueprint /Game/UI/Foundation/Modals/ConfirmationWindow/ConfirmationWindow
animations [Intro {
  binding_count=1
  length_seconds=0.25
  object=&self:Intro
  sequence {bindings [CommonBorder_15 {
     id=#1
     possessed_class=CommonBorder
     tracks [
      {
       type=MovieSceneFloatTrack
       display_name=Render Opacity
       sections [{type=MovieSceneFloatSection, end_frame=5, start_frame=0, channels [None {type=MovieSceneScriptingFloatChannel, keys [[0,0],[2,0],[5,1]]}]}]
      }
      {
       type=MovieScene2DTransformTrack
       sections [{
         type=MovieScene2DTransformSection
         end_frame=5
         start_frame=0
         channels [
          Scale.X {type=MovieSceneScriptingFloatChannel, keys [[2,1.15],[4,1]]}
          Scale.Y {type=MovieSceneScriptingFloatChannel, keys [[2,1.15],[4,1]]}
         ]
        }]
      }
     ]
    }]}
 }]
graphs [
 ConfigureConfirmationButton {nodes [
   n1 = ConfigureConfirmationButton!() -> (then, Button:Common Button Base Object Reference, Action:"Confirmation Dialog Action Structure (by ref)", bSimpleConfirm:bool) {is=title K2Node_FunctionEntry}
   > n2 = Cast To WBP_UIKit_ButtonCTA_Base!(execute, Object=@n1) -> (then, CastFailed, AsWBP UIKit Button CTA Base:WBP UIKit Button CTA Base Object Reference) {is=title K2Node_DynamicCast}
   > n3 = Set CTAButton(CTAButton:WBP UIKit Button CTA Base Object Reference=@n2) -> (Output_Get:WBP UIKit Button CTA Base Object Reference)
   > ResetButtonState(self:WBP UIKit Button CTA Base Object Reference=@n3)
   n4 = Get CTAButton() -> (CTAButton:WBP UIKit Button CTA Base Object Reference)
   > SetInputActionPriority!(
    execute
    self:User Widget Object Reference=@n4
    NewPriority:int=GetInputPriority!(
     self:Fort Global UIContext Object Reference=Get FortGlobalUIContext!() -> (ReturnValue:Fort Global UIContext Object Reference) {is=title K2Node_GetBlueprintContextEx}
     Priority:EInputPriority Enum=Get InputPriority() -> (InputPriority:EInputPriority Enum=Normal)
    ) -> (ReturnValue:int=0)
   ) -> (then)
   n5 = Break Confirmation Dialog Action!(ConfirmationDialogAction:"Confirmation Dialog Action Structure (by ref)"=@n1.Action) -> (DisplayName:Text, ResultName:Name, Icon:Slate Brush Structure, ActionName:Name) {is=title K2Node_BreakStruct}
   > SetTriggeringInputAction!(
    execute
    self:Common Button Base Object Reference=@n4
    InputActionRow:"Data Table Row Handle Structure (by ref)"=GetInputAction!(RowName:Name=@n5.ActionName, UseInputAction:bool=@n1.bSimpleConfirm) -> (ReturnValue:Data Table Row Handle Structure)
   ) -> (then)
   > n6 = Branch(Condition=TextIsEmpty(InText=@n5))
   > n7 = ShowInputActionText(self:WBP UIKit Button CTA Base Object Reference=@n4)
   SetButtonSize!(execute=[@n7.then, @n8.then], self:WBP UIKit Button CTA Base Object Reference=@n4, ButtonSize:E_UI_ButtonSize Enum=NewEnumerator1) -> (then)
   > Return Node()
   n8 = SetText!(execute=@n6.else, self:Fort CTAButton Object Reference=@n4, ButtonText:"Text (by ref)"=@n5) -> (then)
  ]}
 EventGraph {nodes [
   Event PreConstruct()
   > SetMinDesiredWidth!(
    execute
    self:Size Box Object Reference=Get SizeBox_0() -> (SizeBox_0:Size Box Object Reference)
    InMinDesiredWidth:float=Select(Option 0:double=1000, Option 1:double=1250, Index=IsMobileGame()) -> (ReturnValue:double)
   ) -> (then)
   > n1 = For Each Loop(Array:Array of User Widget Object References=GetAllEntries!(self:Dynamic Entry Box Base Object Reference=Get EntryBox_SimpleButtons() -> (EntryBox_SimpleButtons:Dynamic Entry Box Object Reference)) -> (ReturnValue:Array of User Widget Object References)) -> (Array Element:User Widget Object Reference)
   > n2 = Cast To WBP_UIKit_ButtonCTA_Base(Object=@n1)
   > SetButtonSize(self=@n2, ButtonSize=NewEnumerator2)
   Event On Activated()
   > Initialize()
   > BotClickButton(self:Fort Confirmation Window Object Reference)
   > PlayAnimationForward(InAnimation=Get Intro())
   n3 = Event ConfigureDeclineButton!() -> (OutputDelegate, then, Button:Common Button Base Object Reference) {is=title K2Node_Event}
   > Handle Configure Decline Button!(execute, self, New Button:Common Button Base Object Reference=@n3.Button) -> (then)
   Event OnBeginOutro() {is=title K2Node_Event}
   > EndOutro(self:Common Activatable Panel Legacy Object Reference)
   > SpawnSound2D!(
    execute
    Sound:Sound Base Object Reference=Get OutroSound() -> (OutroSound:Sound Base Object Reference)
    VolumeMultiplier:float=1
    PitchMultiplier:float=1
    StartTime:float=0
    ConcurrencySettings:Sound Concurrency Object Reference
    bPersistAcrossLevelTransition:bool=false
    bAutoDestroy:bool=true
   ) -> (then, ReturnValue:Audio Component Object Reference)
  ]}
 GetInputAction {nodes [
   n1 = GetInputAction!() -> (then, RowName:Name, UseInputAction:bool=~) {is=title K2Node_FunctionEntry}
   > Return Node!(
    execute
    ReturnValue:Data Table Row Handle Structure=Select(
     Option 0:Data Table Row Handle Structure
     Option 1:Data Table Row Handle Structure=Make Data Table Row Handle!(DataTable:Data Table Object Reference=&InputActionDataTable, RowName:Name=@n1) -> (DataTableRowHandle:Data Table Row Handle Structure) {is=title K2Node_MakeStruct}
     Index=@n1.UseInputAction
    ) -> (ReturnValue:Data Table Row Handle Structure)
   )
  ]}
 Handle Configure Decline Button {nodes [
   n1 = Handle Configure Decline Button!() -> (then, New Button:Common Button Base Object Reference) {is=title K2Node_FunctionEntry}
   > n2 = Cast To WBP_UIKit_ButtonCTA_Base(Object=@n1)
   > n3 = Set CTAButton(CTAButton=@n2)
   > ResetButtonState(self=@n3)
   n4 = GetConfirmationDescription!(self:Fort Confirmation Window Object Reference) -> (ReturnValue:"Fort Dialog Description NUI Structure (by ref)")
   n5 = Get CTAButton()
   > SetTriggeringInputAction(
    self=@n5
    InputActionRow=Make Data Table Row Handle(
     DataTable=&InputActionDataTable
     RowName=Break Fort Dialog Description NUI!(FortDialogDescription_NUI:"Fort Dialog Description NUI Structure (by ref)"=@n4) -> (
      DeclineButtonInputAction:Name
      Cancelable:bool=~
      bCloseOnEsc:bool=~
      bShouldWaitForLatentActionOnConfirmAction:bool=~
      ShowSound:Sound Base Object Reference
     ) {is=title K2Node_BreakStruct}
    )
   )
   > SetInputActionPriority(
    self=@n5
    NewPriority=GetInputPriority(self=Get FortGlobalUIContext(), Priority=Get InputPriority() -> (InputPriority=Normal)) -> (ReturnValue=0)
   )
   > ShowInputActionText(self=@n5)
   > SetButtonSize(self=@n5, ButtonSize=NewEnumerator1)
  ]}
 Initialize {nodes [
   Initialize() {is=title K2Node_FunctionEntry}
   > PlayShowSound()
  ]}
 PlayShowSound {nodes [
   PlayShowSound() {is=title K2Node_FunctionEntry}
   > PlaySound2D(Sound=Get IntroSound() -> (IntroSound:Sound Base Object Reference))
  ]}
 SetupNonInteractiveContent {nodes [SetupNonInteractiveContent() {is=title K2Node_FunctionEntry}]}
 TapToClose {nodes [
   n1 = TapToClose!() -> (then, MyGeometry:Geometry Structure, MouseEvent:"Pointer Event Structure (by ref)") {is=title K2Node_FunctionEntry}
   > n2 = Handle Primary Pointer Button or Touch!(Execute:Exec, Mouse Event:"Pointer Event Structure (by ref)"=@n1.MouseEvent) -> (If Primary:Exec, Continue:Exec, Reply:Event Reply Structure) {is=title K2Node_MacroInstance}
   Return Node(execute=@n2.Continue, ReturnValue:Event Reply Structure=@n2)
   > HandleDeclineActionClicked!(execute, self:Fort Confirmation Window Object Reference, ButtonClicked:Common Button Base Object Reference) -> (then)
  ]}
]
variables [
 ButtonIconColor {type=LinearColor, default=<rgba 0.008 0.011 0.0262 1>}
 IntroSound {type=SoundBase, default=/Game/Sounds/Fort_UI_Sounds/FrontEnd/Whooshes/Toasts/UI_Whoosh_Generic_Confirmation_Appear_Cue}
 OutroSound {type=SoundBase, default=/Game/Sounds/Fort_UI_Sounds/FrontEnd/Whooshes/Toasts/UI_Whoosh_Generic_Confirmation_Close_Cue}
]
widgets {
 info {
  inheritedWidgetCount=10
  namedSlotCount=1
  parentClass=/Script/FortniteUI.FortConfirmationWindow
  rootWidgetClass=/Game/UI/Foundation/Lightbox/Lightbox.Lightbox_C
  widgetCount=25
 }
 widgets [
  Lightbox_NUI {
   widgetClassPath=/Game/UI/Foundation/Lightbox/Lightbox
   widget=Lightbox_NUI
   properties {dimAmount=0.75, backgroundColor=<rgba 0 0.0006 0.0075 1>}
  }
  Overlay_3 {widgetClassPath=&Overlay, namedSlotHost=Lightbox_NUI, widget=Overlay_3}
  WBP_UIKit_Scrim_ContainsBlur {
   widgetClassPath=/FortUILibrary/Systems/Dialog/WBP_UIKit_Scrim
   bIsVariable=true
   parent=Overlay_3
   widget=WBP_UIKit_Scrim_ContainsBlur
   slot {class=OverlaySlot, properties {horizontalAlignment=HAlign_Fill, verticalAlignment=VAlign_Fill}}
  }
  TapToCloseZone {
   widgetClassPath=&CommonBorder
   parent=Overlay_3
   widget=TapToCloseZone
   properties {onMouseButtonDownEvent="(null).On_TapToCloseZone_MouseButtonDown_0", visibility=Visible, padding=<ltrb 0 0 0 0>}
   slot {class=OverlaySlot, properties {horizontalAlignment=HAlign_Fill, verticalAlignment=VAlign_Fill}}
  }
  SizeBox_0 {
   widgetClassPath=&SizeBox
   bIsVariable=true
   parent=Overlay_3
   widget=SizeBox_0
   properties {
    bOverride_MaxAspectRatio=true
    bOverride_MinDesiredHeight=true
    bOverride_MinDesiredWidth=true
    maxAspectRatio=2.5
    minDesiredHeight=600
    minDesiredWidth=1000
   }
   slot {class=OverlaySlot, properties {horizontalAlignment=HAlign_Center, verticalAlignment=VAlign_Center, padding=<ltrb 100 100 100 100>}}
  }
  CommonBorder_15 {
   widgetClassPath=&CommonBorder
   parent=SizeBox_0
   widget=CommonBorder_15
   properties {style=/FortUILibrary/Styles/Backgrounds/_BorderStyle_Modal._BorderStyle_Modal_C, visibility=Visible, padding=<ltrb 0 0 0 0>}
   slot {class=SizeBoxSlot}
  }
  Overlay_0 {widgetClassPath=&Overlay, parent=CommonBorder_15, widget=Overlay_0, slot {class=BorderSlot, properties {padding=<ltrb 0 0 0 0>}}}
  WBP_CaptureForPostBufferUpdate {
   widgetClassPath=&WBP_CaptureForPostBufferUpdate
   bIsVariable=true
   parent=Overlay_0
   slot {class=OverlaySlot, properties {horizontalAlignment=HAlign_Fill, verticalAlignment=VAlign_Fill}}
  }
  MainBox {
   widgetClassPath=&VerticalBox
   parent=Overlay_0
   widget=MainBox
   properties {navigation=&self:WidgetTree.MainBox.WidgetNavigation_0}
   slot {class=OverlaySlot, properties {horizontalAlignment=HAlign_Fill, verticalAlignment=VAlign_Fill}}
  }
  CommonBorder_Modal {
   widgetClassPath=&CommonBorder
   parent=MainBox
   widget=CommonBorder_Modal
   properties {
    style=/FortUILibrary/Styles/Backgrounds/_BorderStyle_ModalHeader._BorderStyle_ModalHeader_C
    background {
     bIsDynamicallyLoaded=false
     drawAs=Box
     imageType=NoImage
     mirroring=NoMirror
     resourceName=None
     resourceObject=/FortUILibrary/Styles/Border/Materials/M-Border-ModalHeader-procedural
     tiling=NoTile
     imageSize=<64 64>
     margin=<ltrb 0.25 0.25 0.25 0.25>
     outlineSettings {
      bUseBrushTransparency=false
      roundingType=HalfHeightRadius
      width=0
      color {colorUseRule=UseColor_Specified, specifiedColor=<rgba 0 0 0 0>}
      cornerRadii=<0 0 0 0>
     }
     tintColor {colorUseRule=UseColor_Specified, specifiedColor=<rgba 1 1 1 1>}
     uVRegion {bIsValid=false, max=<0 0>, min=<0 0>}
    }
    padding=<ltrb 40 0 4 0>
   }
   slot {class=VerticalBoxSlot}
  }
  HorizontalBox_1 {
   widgetClassPath=&HorizontalBox
   parent=CommonBorder_Modal
   widget=HorizontalBox_1
   slot {class=BorderSlot, properties {padding=<ltrb 40 0 4 0>}}
  }
  Spacer_89 {
   widgetClassPath=&Spacer
   parent=HorizontalBox_1
   widget=Spacer_89
   properties {size=<0 85>}
   slot {class=HorizontalBoxSlot, properties {horizontalAlignment=HAlign_Left, verticalAlignment=VAlign_Top}}
  }
  Image_MainIcon {
   widgetClassPath=&Image
   bInherited=true
   bIsVariable=true
   parent=HorizontalBox_1
   widget=Image_MainIcon
   properties {brush {
     bIsDynamicallyLoaded=false
     drawAs=Image
     imageType=NoImage
     mirroring=NoMirror
     resourceName=None
     resourceObject=null
     tiling=NoTile
     imageSize=<64 64>
     margin=<ltrb 0 0 0 0>
     outlineSettings {
      bUseBrushTransparency=false
      roundingType=HalfHeightRadius
      width=0
      color {colorUseRule=UseColor_Specified, specifiedColor=<rgba 0 0 0 0>}
      cornerRadii=<0 0 0 0>
     }
     tintColor {colorUseRule=UseColor_Specified, specifiedColor=<rgba 1 1 1 1>}
     uVRegion {bIsValid=false, max=<0 0>, min=<0 0>}
    }}
   slot {class=HorizontalBoxSlot, properties {horizontalAlignment=HAlign_Left, verticalAlignment=VAlign_Center, padding=<ltrb 20 0 0 0>}}
  }
  CommonTextBlock_Title {
   widgetClassPath=&CommonTextBlock
   bInherited=true
   bIsVariable=true
   parent=HorizontalBox_1
   widget=CommonTextBlock_Title
   properties {
    clipping=ClipToBounds
    scrollStyle=&TextScrollStyle-Base_C
    style=/FortUILibrary/Styles/Typography/_TextStyle_SubH2._TextStyle_SubH2_C
    text=Title
    textTransformPolicy=ToUpper
    font {
     bForceMonospaced=false
     bMaterialIsStencil=false
     fontMaterial=null
     fontObject=&Burbank
     letterSpacing=0
     monospacedWidth=1
     size=33
     skewAmount=0
     typefaceFontName=BurbankRegular-Black
     outlineSettings {
      bApplyOutlineToDropShadows=false
      bMiteredCorners=false
      bSeparateFillAlpha=false
      outlineMaterial=null
      outlineSize=0
      outlineColor=<rgba 0 0 0 1>
     }
    }
    shadowOffset=<0 0>
    strikeBrush {
     bIsDynamicallyLoaded=false
     drawAs=NoDrawType
     imageType=NoImage
     mirroring=NoMirror
     resourceName=None
     resourceObject=null
     tiling=NoTile
     imageSize=<32 32>
     margin=<ltrb 0 0 0 0>
     outlineSettings {
      bUseBrushTransparency=false
      roundingType=HalfHeightRadius
      width=0
      color {colorUseRule=UseColor_Specified, specifiedColor=<rgba 0 0 0 0>}
      cornerRadii=<0 0 0 0>
     }
     tintColor {colorUseRule=UseColor_Specified, specifiedColor=<rgba 1 1 1 1>}
     uVRegion {bIsValid=false, max=<0 0>, min=<0 0>}
    }
   }
   slot {class=HorizontalBoxSlot, properties {verticalAlignment=VAlign_Center, padding=<ltrb 8 0 20 0>, size {sizeRule=Fill, value=1}}}
  }
  CommonLoadGuard_LatentSpinner {
   widgetClassPath=/Script/CommonUI.CommonLoadGuard
   bInherited=true
   parent=MainBox
   widget=CommonLoadGuard_LatentSpinner
   properties {loadingText="", textStyle=&TextStyle-Base_C}
   slot {class=VerticalBoxSlot, properties {size {sizeRule=Fill, value=1}}}
  }
  MainContentHorizontalBox {widgetClassPath=&HorizontalBox, parent=CommonLoadGuard_LatentSpinner, widget=MainContentHorizontalBox, slot {class=LoadGuardSlot}}
  Panel_LeftOptionalWidget {
   widgetClassPath=&HorizontalBox
   bInherited=true
   parent=MainContentHorizontalBox
   widget=Panel_LeftOptionalWidget
   slot {class=HorizontalBoxSlot, properties {horizontalAlignment=HAlign_Left, verticalAlignment=VAlign_Top, padding=<ltrb 44 20 44 44>}}
  }
  TextItemContent {
   widgetClassPath=&VerticalBox
   parent=MainContentHorizontalBox
   widget=TextItemContent
   slot {class=HorizontalBoxSlot, properties {padding=<ltrb 48 0 48 0>, size {sizeRule=Fill, value=1}}}
  }
  ScrollBox_Description {
   widgetClassPath=&ScrollBox
   bInherited=true
   bIsVariable=true
   parent=TextItemContent
   widget=ScrollBox_Description
   properties {
    scrollbarThickness=<20 10>
    widgetBarStyle {
     thickness=16
     draggedThumbImage {
      bIsDynamicallyLoaded=false
      drawAs=Box
      imageType=FullColor
      mirroring=NoMirror
      resourceName=../../../Engine/Content/Slate/Common/Scrollbar_Thumb.png
      resourceObject=null
      tiling=NoTile
      imageSize=<0 0>
      margin=<ltrb 0.25 0.25 0.25 0.25>
      outlineSettings {
       bUseBrushTransparency=false
       roundingType=HalfHeightRadius
       width=0
       color {colorUseRule=UseColor_Specified, specifiedColor=<rgba 0 0 0 0>}
       cornerRadii=<0 0 0 0>
      }
      tintColor {colorUseRule=UseColor_Specified, specifiedColor=<rgba 1 1 1 1>}
      uVRegion {bIsValid=false, max=<0 0>, min=<0 0>}
     }
     horizontalBackgroundImage {
      bIsDynamicallyLoaded=false
      drawAs=NoDrawType
      imageType=NoImage
      mirroring=NoMirror
      resourceName=None
      resourceObject=null
      tiling=NoTile
      imageSize=<0 0>
      margin=<ltrb 0 0 0 0>
      outlineSettings {
       bUseBrushTransparency=false
       roundingType=HalfHeightRadius
       width=0
       color {colorUseRule=UseColor_Specified, specifiedColor=<rgba 0 0 0 0>}
       cornerRadii=<0 0 0 0>
      }
      tintColor {colorUseRule=UseColor_Specified, specifiedColor=<rgba 1 1 1 1>}
      uVRegion {bIsValid=false, max=<0 0>, min=<0 0>}
     }
     horizontalBottomSlotImage {
      bIsDynamicallyLoaded=false
      drawAs=Image
      imageType=FullColor
      mirroring=NoMirror
      resourceName=../../../Engine/Content/Slate/Common/Scrollbar_Background_Horizontal.png
      resourceObject=null
      tiling=NoTile
      imageSize=<8 8>
      margin=<ltrb 0 0 0 0>
      outlineSettings {
       bUseBrushTransparency=false
       roundingType=HalfHeightRadius
       width=0
       color {colorUseRule=UseColor_Specified, specifiedColor=<rgba 0 0 0 0>}
       cornerRadii=<0 0 0 0>
      }
      tintColor {colorUseRule=UseColor_Specified, specifiedColor=<rgba 1 1 1 1>}
      uVRegion {bIsValid=false, max=<0 0>, min=<0 0>}
     }
     horizontalTopSlotImage {
      bIsDynamicallyLoaded=false
      drawAs=Image
      imageType=FullColor
      mirroring=NoMirror
      resourceName=../../../Engine/Content/Slate/Common/Scrollbar_Background_Horizontal.png
      resourceObject=null
      tiling=NoTile
      imageSize=<8 8>
      margin=<ltrb 0 0 0 0>
      outlineSettings {
       bUseBrushTransparency=false
       roundingType=HalfHeightRadius
       width=0
       color {colorUseRule=UseColor_Specified, specifiedColor=<rgba 0 0 0 0>}
       cornerRadii=<0 0 0 0>
      }
      tintColor {colorUseRule=UseColor_Specified, specifiedColor=<rgba 1 1 1 1>}
      uVRegion {bIsValid=false, max=<0 0>, min=<0 0>}
     }
     hoveredThumbImage {
      bIsDynamicallyLoaded=false
      drawAs=Box
      imageType=FullColor
      mirroring=NoMirror
      resourceName=../../../Engine/Content/Slate/Common/Scrollbar_Thumb.png
      resourceObject=null
      tiling=NoTile
      imageSize=<0 0>
      margin=<ltrb 0.25 0.25 0.25 0.25>
      outlineSettings {
       bUseBrushTransparency=false
       roundingType=HalfHeightRadius
       width=0
       color {colorUseRule=UseColor_Specified, specifiedColor=<rgba 0 0 0 0>}
       cornerRadii=<0 0 0 0>
      }
      tintColor {colorUseRule=UseColor_Specified, specifiedColor=<rgba 1 1 1 1>}
      uVRegion {bIsValid=false, max=<0 0>, min=<0 0>}
     }
     normalThumbImage {
      bIsDynamicallyLoaded=false
      drawAs=Box
      imageType=FullColor
      mirroring=NoMirror
      resourceName=../../../Engine/Content/Slate/Common/Scrollbar_Thumb.png
      resourceObject=/FortUILibrary/Foundation/Buttons/Materials/Skewed/M-Button-Skewed2-Base
      tiling=NoTile
      imageSize=<0 0>
      margin=<ltrb 0.25 0.25 0.25 0.25>
      outlineSettings {
       bUseBrushTransparency=false
       roundingType=HalfHeightRadius
       width=0
       color {colorUseRule=UseColor_Specified, specifiedColor=<rgba 0 0 0 0>}
       cornerRadii=<0 0 0 0>
      }
      tintColor {colorUseRule=UseColor_Specified, specifiedColor=<rgba 1 1 1 1>}
      uVRegion {bIsValid=false, max=<0 0>, min=<0 0>}
     }
     verticalBackgroundImage {
      bIsDynamicallyLoaded=false
      drawAs=NoDrawType
      imageType=NoImage
      mirroring=NoMirror
      resourceName=None
      resourceObject=null
      tiling=NoTile
      imageSize=<0 0>
      margin=<ltrb 0 0 0 0>
      outlineSettings {
       bUseBrushTransparency=false
       roundingType=HalfHeightRadius
       width=0
       color {colorUseRule=UseColor_Specified, specifiedColor=<rgba 0 0 0 0>}
       cornerRadii=<0 0 0 0>
      }
      tintColor {colorUseRule=UseColor_Specified, specifiedColor=<rgba 1 1 1 1>}
      uVRegion {bIsValid=false, max=<0 0>, min=<0 0>}
     }
     verticalBottomSlotImage {
      bIsDynamicallyLoaded=false
      drawAs=Image
      imageType=FullColor
      mirroring=NoMirror
      resourceName=../../../Engine/Content/Slate/Common/Scrollbar_Background_Vertical.png
      resourceObject=null
      tiling=NoTile
      imageSize=<8 8>
      margin=<ltrb 0 0 0 0>
      outlineSettings {
       bUseBrushTransparency=false
       roundingType=HalfHeightRadius
       width=0
       color {colorUseRule=UseColor_Specified, specifiedColor=<rgba 0 0 0 0>}
       cornerRadii=<0 0 0 0>
      }
      tintColor {colorUseRule=UseColor_Specified, specifiedColor=<rgba 1 1 1 1>}
      uVRegion {bIsValid=false, max=<0 0>, min=<0 0>}
     }
     verticalTopSlotImage {
      bIsDynamicallyLoaded=false
      drawAs=Image
      imageType=FullColor
      mirroring=NoMirror
      resourceName=../../../Engine/Content/Slate/Common/Scrollbar_Background_Vertical.png
      resourceObject=null
      tiling=NoTile
      imageSize=<8 8>
      margin=<ltrb 0 0 0 0>
      outlineSettings {
       bUseBrushTransparency=false
       roundingType=HalfHeightRadius
       width=0
       color {colorUseRule=UseColor_Specified, specifiedColor=<rgba 0 0 0 0>}
       cornerRadii=<0 0 0 0>
      }
      tintColor {colorUseRule=UseColor_Specified, specifiedColor=<rgba 1 1 1 1>}
      uVRegion {bIsValid=false, max=<0 0>, min=<0 0>}
     }
    }
   }
   slot {class=VerticalBoxSlot, properties {padding=<ltrb 0 34 0 0>, size {sizeRule=Fill, value=1}}}
  }
  CommonTextBlock_Description {
   widgetClassPath=&CommonTextBlock
   bInherited=true
   bIsVariable=true
   parent=ScrollBox_Description
   widget=CommonTextBlock_Description
   properties {
    autoWrapText=true
    lineHeightPercentage=1.29
    mobileFontSizeMultiplier=1.3
    style=/FortUILibrary/Styles/Typography/_TextStyle_Body1._TextStyle_Body1_C
    text="[Description text]"
    font {
     bForceMonospaced=false
     bMaterialIsStencil=false
     fontMaterial=null
     fontObject=&BurbankSmall
     letterSpacing=0
     monospacedWidth=1
     size=18
     skewAmount=0
     typefaceFontName=Bold
     outlineSettings {
      bApplyOutlineToDropShadows=false
      bMiteredCorners=false
      bSeparateFillAlpha=false
      outlineMaterial=null
      outlineSize=0
      outlineColor=<rgba 0 0 0 1>
     }
    }
    shadowOffset=<0 0>
    strikeBrush {
     bIsDynamicallyLoaded=false
     drawAs=NoDrawType
     imageType=NoImage
     mirroring=NoMirror
     resourceName=None
     resourceObject=null
     tiling=NoTile
     imageSize=<32 32>
     margin=<ltrb 0 0 0 0>
     outlineSettings {
      bUseBrushTransparency=false
      roundingType=HalfHeightRadius
      width=0
      color {colorUseRule=UseColor_Specified, specifiedColor=<rgba 0 0 0 0>}
      cornerRadii=<0 0 0 0>
     }
     tintColor {colorUseRule=UseColor_Specified, specifiedColor=<rgba 1 1 1 1>}
     uVRegion {bIsValid=false, max=<0 0>, min=<0 0>}
    }
   }
   slot {class=ScrollBoxSlot}
  }
  SizeBox_1 {
   widgetClassPath=&SizeBox
   parent=TextItemContent
   widget=SizeBox_1
   properties {bOverride_MaxDesiredWidth=true, bOverride_MinDesiredWidth=true, maxDesiredWidth=660, minDesiredWidth=500}
   slot {class=VerticalBoxSlot, properties {horizontalAlignment=HAlign_Center, padding=<ltrb 0 40 0 0>}}
  }
  EntryBox_ComplexChoiceButtons {
   widgetClassPath=&DynamicEntryBox
   bInherited=true
   bIsVariable=true
   parent=SizeBox_1
   widget=EntryBox_ComplexChoiceButtons
   properties {
    entryBoxType=Vertical
    entryVerticalAlignment=VAlign_Bottom
    entryWidgetClass=/FortUILibrary/UIKit/Buttons/WBP_UIKit_Button_Regular.WBP_UIKit_Button_Regular_C
    navigation=&self:WidgetTree.EntryBox_ComplexChoiceButtons.WidgetNavigation_0
    entrySpacing=<0 16>
   }
   slot {class=SizeBoxSlot, properties {verticalAlignment=VAlign_Bottom}}
  }
  Panel_OptionalWidget {
   widgetClassPath=&CommonBorder
   bInherited=true
   parent=TextItemContent
   widget=Panel_OptionalWidget
   properties {padding=<ltrb 0 0 0 0>}
   slot {class=VerticalBoxSlot, properties {verticalAlignment=VAlign_Bottom, padding=<ltrb 0 0 0 32>}}
  }
  EntryBox_SimpleButtons {
   widgetClassPath=&DynamicEntryBox
   bInherited=true
   bIsVariable=true
   parent=MainBox
   widget=EntryBox_SimpleButtons
   properties {
    entryVerticalAlignment=VAlign_Bottom
    entryWidgetClass=&WBP_UIKit_Button_Regular_C
    numDesignerPreviewEntries=2
    entrySizeRule {sizeRule=Fill, value=1}
    entrySpacing=<48 0>
   }
   slot {class=VerticalBoxSlot, properties {padding=<ltrb 48 0 48 32>}}
  }
  Button_CloseTouch {widgetClassPath=/Script/UMG.Widget, bInherited=true, widget=null}
 ]
}
